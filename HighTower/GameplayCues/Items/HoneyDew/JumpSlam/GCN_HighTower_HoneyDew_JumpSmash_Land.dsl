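; Burst gameplay cue notify that dresses the landing of the HighTower HoneyDew item's jump slam with particles, a decal and sound.
; On burst it nudges a spawned Niagara component by a local offset and pitches the burst decal 90 degrees.
; It moves the decal to the particle component's world location and plays a sound there, picked from the burst sounds by a truncated cue-parameter float.
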(fn OnBurst(Niagara)
  (params (MyTarget :type Actor%20Object%20Reference) (Parameters :type Gameplay%20Cue%20Parameters%20Structure%20%28by%20ref%29) (ParticleComponents :type Array%20of%20Niagara%20Particle%20System%20Component%20Object%20References) (AudioComponents :type Array%20of%20Audio%20Component%20Object%20References) (BurstCameraShakeInstance :type Camera%20Shake%20Base%20Object%20Reference) (BurstDecalInstance :type Decal%20Actor%20Object%20Reference))
  (bind _output (Utilities|Array|Get(acopy) ParticleComponents))
  (bind _returnvalue (Transformation|GetWorldLocation _output))
  (bind _rootcomponent (Variables|/Script/Engine.Actor|GetRootComponent BurstDecalInstance))
  (Transformation|AddLocalOffset _output "-20.000000,31.000000, 0" true true)
  (Transformation|SetRelativeRotation _rootcomponent "90.000000, 0, 0" false false)
  (Transformation|SetWorldLocation _rootcomponent _returnvalue false false)
  (Audio|PlaySoundatLocation 0 (Utilities|Array|Get(acopy) (LocalVariables|GetBurstSounds) (Math|Float|Truncate (Ability|GameplayCue|BreakGameplayCueParameters Parameters))) _returnvalue)
  (return))
